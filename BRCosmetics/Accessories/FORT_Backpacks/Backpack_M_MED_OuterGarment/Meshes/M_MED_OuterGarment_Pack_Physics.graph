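# Physics asset for the M_MED_OuterGarment backpack cosmetic, giving the pack mesh a simple collision and constraint setup.
# Two thin, similarly sized boxes cover the `pack` bone and the `dyn_backpack_c` bone, with the latter box slightly rotated.
# A single constraint joins the two bodies with both swing axes and twist locked, so `dyn_backpack_c` cannot rotate relative to `pack` through that joint.
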PhysicsAsset /BRCosmetics/Accessories/FORT_Backpacks/Backpack_M_MED_OuterGarment/Meshes/M_MED_OuterGarment_Pack_Physics
bodies {
 dyn_backpack_c [{shapeType=Box, extentX=24, extentY=3.5, extentZ=28.5, shapeName=dyn_backpack_c_box, center=<0 -0.2517 -0.6817>, rotation=<rot 0 0 3.5809>}]
 pack [{shapeType=Box, extentX=24, extentY=2, extentZ=29, shapeName=root_box, center=<0 0.4953 -1.293>}]
}
constraints [{
  bone1Name=dyn_backpack_c
  bone2Name=pack
  swing1LimitDegrees=45
  swing1Motion=Locked
  swing2LimitDegrees=45
  swing2Motion=Locked
  twistLimitDegrees=45
  twistMotion=Locked
 }]
