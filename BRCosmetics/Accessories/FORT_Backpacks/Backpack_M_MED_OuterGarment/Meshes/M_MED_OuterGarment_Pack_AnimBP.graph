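# Animation Blueprint for the OuterGarment backpack, adding procedural secondary motion to the pack and its side "cheek" bones.
# The AnimGraph chains pack offsets (one jump-driven), Anim Dynamics springs on pack and cheeks, additive bone translations, and a final RigidBody simulation.
# Each update, real-time sine waves drive the pack and opposite-signed left/right cheek wobble.
# Crouch, sprint and skydive/jump branches retune the pack simulation's velocity, acceleration and gravity inputs.
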AnimBlueprint /BRCosmetics/Accessories/FORT_Backpacks/Backpack_M_MED_OuterGarment/Meshes/M_MED_OuterGarment_Pack_AnimBP
target_skeleton=&/M_MED_OuterGarment_Pack_Skeleton
graphs [
 AnimGraph {nodes [
   n1 = "Transform (Modify) Bone - Bone: pack"(ComponentPose=Local To Component(LocalPose=Input Pose()), Translation="0.000,-3.572,0.678", Rotation="0,0,0") {
    translationMode=BMM_Additive
    translationSpace=BCS_BoneSpace
    alphaBoolBlend {blendInTime=0.5, blendOption=Linear, blendOutTime=1}
    boneToModify {boneName=pack}
    translation {y=-3.5719, z=0.678}
   }
   n2 = "Transform (Modify) Bone - Bone: pack"!(ComponentPose=@n1, bAlphaBoolEnabled:bool=Get bIsJumping(), Translation="0,-3,0.6", Rotation="0,0,0", Scale) -> (Pose) {
    alphaInputType=Bool
    translationMode=BMM_Additive
    translationSpace=BCS_BoneSpace
    alphaBoolBlend {blendInTime=0.2, blendOption=Linear, blendOutTime=1}
    boneToModify {boneName=pack}
    translation {y=-3, z=0.6}
   }
   n3 = Anim Dynamics - Bone: pack!(
    ComponentPose=@n2
    Alpha=1
    GravityOverride:Vector=Get Gravity_Override() -> (Gravity_Override:Vector)
    ComponentLinearAccScale:Vector=Get Linear_Acc_Scale() -> (Linear_Acc_Scale:Vector)
    ComponentLinearVelScale:Vector=Get Linear_Vel_Scale() -> (Linear_Vel_Scale:Vector)
    ExternalForce="0,0,0"
   ) -> (Pose) {
    bGravityOverrideInSimSpace=true
    bLinearSpring=true
    bUseGravityOverride=true
    linearSpringConstant=20
    simulationSpace=BoneRelative
    alphaBoolBlend {blendOption=Linear}
    alphaScaleBiasClamp {bInterpResult=true, interpSpeedDecreasing=500, interpSpeedIncreasing=500}
    boundBone {boneName=pack}
    componentAppliedLinearAccClamp=<50 60 130>
    physicsBodyDefinitions [{
      boundBone {boneName=pack}
      boxExtents=<20 8 20>
      constraintSetup {angularLimitsMax=<0 0 0>, angularLimitsMin=<0 0 0>, angularTarget=<0 0 0>}
      localJointOffset=<0 0 0>
     }]
    relativeSpaceBone {boneName=root}
   }
   n4 = "Transform (Modify) Bone - Bone: pack"!(ComponentPose=@n3, Alpha:float, Translation=Get dyn_pack_Float() -> (dyn_pack_Float:Vector), Rotation="0,0,0", Scale) -> (Pose) {translationMode=BMM_Additive, translationSpace=BCS_BoneSpace, alphaBoolBlend {blendOption=Linear}, boneToModify {boneName=pack}}
   n5 = Anim Dynamics - Bone: dyn_сheek_01_end_l(ComponentPose=@n4, Alpha=0.4, ExternalForce="0,0,0") {
    bGravityOverrideInSimSpace=true
    bLinearSpring=true
    bOverrideLinearDamping=true
    bUseGravityOverride=true
    linearDampingOverride=1
    linearSpringConstant=200
    alphaBoolBlend {blendOption=Linear}
    boundBone {boneName=dyn_сheek_01_end_l}
    componentLinearAccScale=<0.2 0.2 0.2>
    gravityOverride {z=2}
    physicsBodyDefinitions [{
      boundBone {boneName=dyn_сheek_01_end_l}
      boxExtents=<10 10 10>
      constraintSetup {angularLimitsMax=<0 0 0>, angularLimitsMin=<0 0 0>, angularTarget=<0 0 0>, linearAxesMax=<0 0 0.3>, linearAxesMin=<0 0 -1>}
      localJointOffset=<0 0 0>
     }]
   }
   n6 = Anim Dynamics - Bone: dyn_сheek_02_end_l(ComponentPose=@n5, Alpha=0.6, ExternalForce="0,0,0") {
    bGravityOverrideInSimSpace=true
    bLinearSpring=true
    bOverrideLinearDamping=true
    bUseGravityOverride=true
    linearDampingOverride=1
    linearSpringConstant=200
    alphaBoolBlend {blendOption=Linear}
    boundBone {boneName=dyn_сheek_02_end_l}
    componentLinearAccScale=<0.2 0.2 0.2>
    gravityOverride {z=2}
    physicsBodyDefinitions [{
      boundBone {boneName=dyn_сheek_02_end_l}
      boxExtents=<10 10 10>
      constraintSetup {angularLimitsMax=<0 0 0>, angularLimitsMin=<0 0 0>, angularTarget=<0 0 0>, linearAxesMax=<0 0 0.3>, linearAxesMin=<0 0 -0.5>}
      localJointOffset=<0 0 0>
     }]
   }
   n7 = Anim Dynamics - Bone: dyn_сheek_03_end_l(ComponentPose=@n6, Alpha=0.7, ExternalForce="0,0,0") {
    bGravityOverrideInSimSpace=true
    bLinearSpring=true
    bOverrideLinearDamping=true
    bUseGravityOverride=true
    linearDampingOverride=1
    linearSpringConstant=200
    alphaBoolBlend {blendOption=Linear}
    boundBone {boneName=dyn_сheek_03_end_l}
    componentLinearAccScale=<0.2 0.2 0.2>
    gravityOverride {z=1}
    physicsBodyDefinitions [{
      boundBone {boneName=dyn_сheek_03_end_l}
      boxExtents=<10 10 10>
      constraintSetup {angularLimitsMax=<0 0 0>, angularLimitsMin=<0 0 0>, angularTarget=<0 0 0>, linearAxesMax=<0 0 0.1>, linearAxesMin=<0 0 -0.2>}
      localJointOffset=<0 0 0>
     }]
   }
   n8 = Anim Dynamics - Bone: dyn_сheek_01_end_r(ComponentPose=@n7, Alpha=0.5, ExternalForce="0,0,0") {
    bGravityOverrideInSimSpace=true
    bLinearSpring=true
    bOverrideLinearDamping=true
    bUseGravityOverride=true
    linearDampingOverride=1
    linearSpringConstant=200
    alphaBoolBlend {blendOption=Linear}
    boundBone {boneName=dyn_сheek_01_end_r}
    componentLinearAccScale=<0.2 0.2 0.2>
    gravityOverride {z=2}
    physicsBodyDefinitions [{
      boundBone {boneName=dyn_сheek_01_end_r}
      boxExtents=<10 10 10>
      constraintSetup {angularLimitsMax=<0 0 0>, angularLimitsMin=<0 0 0>, angularTarget=<0 0 0>, linearAxesMax=<0 0 0.3>, linearAxesMin=<0 0 -1>}
      localJointOffset=<0 0 0>
     }]
   }
   n9 = Anim Dynamics - Bone: dyn_сheek_02_end_r(ComponentPose=@n8, Alpha=0.6, ExternalForce="0,0,0") {
    bGravityOverrideInSimSpace=true
    bLinearSpring=true
    bOverrideLinearDamping=true
    bUseGravityOverride=true
    linearDampingOverride=1
    linearSpringConstant=200
    alphaBoolBlend {blendOption=Linear}
    boundBone {boneName=dyn_сheek_02_end_r}
    componentLinearAccScale=<0.2 0.2 0.2>
    gravityOverride {z=2}
    physicsBodyDefinitions [{
      boundBone {boneName=dyn_сheek_02_end_r}
      boxExtents=<10 10 10>
      constraintSetup {angularLimitsMax=<0 0 0>, angularLimitsMin=<0 0 0>, angularTarget=<0 0 0>, linearAxesMax=<0 0 0.3>, linearAxesMin=<0 0 -0.5>}
      localJointOffset=<0 0 0>
     }]
   }
   n10 = Anim Dynamics - Bone: dyn_сheek_03_end_r(ComponentPose=@n9, Alpha=0.7, ExternalForce="0,0,0") {
    bGravityOverrideInSimSpace=true
    bLinearSpring=true
    bOverrideLinearDamping=true
    bUseGravityOverride=true
    linearDampingOverride=1
    linearSpringConstant=200
    alphaBoolBlend {blendOption=Linear}
    boundBone {boneName=dyn_сheek_03_end_r}
    componentLinearAccScale=<0.2 0.2 0.2>
    gravityOverride {z=1}
    physicsBodyDefinitions [{
      boundBone {boneName=dyn_сheek_03_end_r}
      boxExtents=<10 10 10>
      constraintSetup {angularLimitsMax=<0 0 0>, angularLimitsMin=<0 0 0>, angularTarget=<0 0 0>, linearAxesMax=<0 0 0.1>, linearAxesMin=<0 0 -0.2>}
      localJointOffset=<0 0 0>
     }]
   }
   n11 = "Transform (Modify) Bone - Bone: dyn_сheek_01_l"(ComponentPose=@n10, Alpha=1, Translation=Get Move_cheek_01_l() -> (Move_cheek_01_l:Vector), Rotation="0,0,0", Scale="1,1,1") {translationMode=BMM_Additive, translationSpace=BCS_BoneSpace, alphaBoolBlend {blendOption=Linear}, boneToModify {boneName=dyn_сheek_01_l}}
   n12 = "Transform (Modify) Bone - Bone: dyn_сheek_02_l"(ComponentPose=@n11, Alpha=1, Translation=Get Move_cheek_02_l() -> (Move_cheek_02_l:Vector), Rotation="0,0,0", Scale="1,1,1") {translationMode=BMM_Additive, translationSpace=BCS_BoneSpace, alphaBoolBlend {blendOption=Linear}, boneToModify {boneName=dyn_сheek_02_l}}
   n13 = "Transform (Modify) Bone - Bone: dyn_сheek_03_l"(ComponentPose=@n12, Alpha=1, Translation=Get Move_cheek_03_l() -> (Move_cheek_03_l:Vector), Rotation="0,0,0", Scale="1,1,1") {translationMode=BMM_Additive, translationSpace=BCS_BoneSpace, alphaBoolBlend {blendOption=Linear}, boneToModify {boneName=dyn_сheek_03_l}}
   n14 = "Transform (Modify) Bone - Bone: dyn_сheek_01_r"(ComponentPose=@n13, Alpha=1, Translation=Get Move_cheek_01_r() -> (Move_cheek_01_r:Vector), Rotation="0,0,0", Scale="1,1,1") {translationMode=BMM_Additive, translationSpace=BCS_BoneSpace, alphaBoolBlend {blendOption=Linear}, boneToModify {boneName=dyn_сheek_01_r}}
   n15 = "Transform (Modify) Bone - Bone: dyn_сheek_02_r"(ComponentPose=@n14, Alpha=1, Translation=Get Move_cheek_02_r() -> (Move_cheek_02_r:Vector), Rotation="0,0,0", Scale="1,1,1") {translationMode=BMM_Additive, translationSpace=BCS_BoneSpace, alphaBoolBlend {blendOption=Linear}, boneToModify {boneName=dyn_сheek_02_r}}
   n16 = "Transform (Modify) Bone - Bone: dyn_сheek_03_r"(ComponentPose=@n15, Alpha=1, Translation=Get Move_cheek_03_r() -> (Move_cheek_03_r:Vector), Rotation="0,0,0", Scale="1,1,1") {translationMode=BMM_Additive, translationSpace=BCS_BoneSpace, alphaBoolBlend {blendOption=Linear}, boneToModify {boneName=dyn_сheek_03_r}}
   n17 = RigidBody(ComponentPose=@n16) {
    overridePhysicsAsset=&/M_MED_OuterGarment_Pack_Physics
    simulationSpace=BaseBoneSpace
    alphaBoolBlend {blendOption=Linear}
    baseBoneRef {boneName=root}
    componentAppliedLinearAccClamp {y=100}
    componentLinearAccScale=<0.1 0.1 0.1>
    componentLinearVelScale=<0.1 0.1 0.1>
    simSpaceSettings {maxAngularAcceleration=10, maxAngularVelocity=25, maxLinearAcceleration=30, maxLinearVelocity=20, worldAlpha=0.75}
   }
   Output Pose(Result=Component To Local(ComponentPose=@n17)) {layerGroup=None}
  ]}
 EventGraph {nodes [
   Event BlueprintUpdateAnimation()
   > n1 = Sequence!(execute) -> (then_0, then_1, then_2:Exec)
   n2 = GetRealTimeSeconds()
   n3 = "Sin (Radians)"(A=float * float(A=float * float(A=1.11, B=1.11), B=@n2))
   n4 = "Sin (Radians)"(A=float * float(A=float * float(A=1.24, B=0.8), B=@n2))
   n5 = "Sin (Radians)"(A=float * float(A=float * float(A=1.8, B=0.5), B=@n2))
   > Set dyn_pack_Float(dyn_pack_Float:Vector=MakeVector(X=float * float(A=@n4, B=2.3), Y=float * float(A=@n3, B=0.9), Z=float * float(A=@n5, B=2))) -> (Output_Get:Vector)
   > Set Move_cheek_03_l(Move_cheek_03_l:Vector=vector * vector(A=Get dyn_pack_Float(), B:Vector="0,0,0.1") -> (ReturnValue=~)) -> (Output_Get:Vector)
   > Set Move_cheek_02_l(Move_cheek_02_l:Vector=vector * vector(A=Get dyn_pack_Float(), B="0,0,0.2") -> (ReturnValue=~)) -> (Output_Get:Vector)
   > Set Move_cheek_01_l(Move_cheek_01_l:Vector=vector * vector(A=Get dyn_pack_Float(), B="0,0,0.1") -> (ReturnValue=~)) -> (Output_Get:Vector)
   > Set Move_cheek_01_r(Move_cheek_01_r:Vector=vector * vector(A=Get dyn_pack_Float(), B="0,0,-0.1") -> (ReturnValue=~)) -> (Output_Get:Vector)
   > Set Move_cheek_02_r(Move_cheek_02_r:Vector=vector * vector(A=Get dyn_pack_Float(), B="0,0,-0.2") -> (ReturnValue=~)) -> (Output_Get:Vector)
   > Set Move_cheek_03_r(Move_cheek_03_r:Vector=vector * vector(A=Get dyn_pack_Float(), B="0,0,-0.2") -> (ReturnValue=~)) -> (Output_Get:Vector)
   n6 = Branch(execute=@n1.then_1, Condition=Get bIsCrouching())
   > Set Linear_Acc_Scale(Linear_Acc_Scale:Vector="0, 0, 0") -> (Output_Get:Vector)
   > Set Linear_Vel_Scale(Linear_Vel_Scale:Vector="10,11,13") -> (Output_Get:Vector)
   Set Linear_Acc_Scale(execute=@n6.else, Linear_Acc_Scale="0.1,0.1,0.1")
   > Set Linear_Vel_Scale(Linear_Vel_Scale="0, 0, 0")
   > Set Gravity_Override(Gravity_Override:Vector="0, 0, 0") -> (Output_Get:Vector)
   n7 = Get bIsCrouching()
   > n8 = Branch(Condition=AND Boolean(A=Get bIsSprinting(), B=NOT Boolean(A=@n7)))
   > Set Linear_Acc_Scale(Linear_Acc_Scale="0.1,0.1,0.1")
   > Set Linear_Vel_Scale(Linear_Vel_Scale="9,11,13")
   Set Linear_Acc_Scale(execute=@n8.else, Linear_Acc_Scale="0.1,0.1,0.1")
   > Set Linear_Vel_Scale(Linear_Vel_Scale="9,250,13")
   > Set Gravity_Override(Gravity_Override="0, 0, 0")
   > n9 = Branch(Condition=OR Boolean(A=Get bIsSkyDiving(), B=Get bIsJumping()))
   > Set Linear_Acc_Scale(Linear_Acc_Scale="0.1,0.1,0.1")
   > Set Linear_Vel_Scale(Linear_Vel_Scale="1.2,1.2,0.6")
   > Set Gravity_Override(Gravity_Override="0,-100,0")
   Set Linear_Acc_Scale(execute=@n9.else, Linear_Acc_Scale="0.1,0.1,0.1")
   > Set Linear_Vel_Scale(Linear_Vel_Scale="0,0,1.3")
   > Set Gravity_Override(Gravity_Override="0, 0, 0")
   TryGetPawnOwner()
  ]}
]
variables [
 dyn_pack_Float {type=Vector}
 Gravity_Override {type=Vector}
 Linear_Acc_Scale {type=Vector}
 Linear_Vel_Scale {type=Vector}
 Move_cheek_01_l {type=Vector}
 Move_cheek_01_r {type=Vector}
 Move_cheek_02_l {type=Vector}
 Move_cheek_02_r {type=Vector}
 Move_cheek_03_l {type=Vector}
 Move_cheek_03_r {type=Vector}
]
